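# Sound cue for the first-person foley of the Journey emote's blade equip, playing a random blade-extend sound.
# A quality-level root sends its first two levels to an equally weighted random pick of six waves and its last two to a two-wave subset.
# Emote attenuation, sound class and property sheet apply, a concurrency override stops the oldest instance, and a submix send targets music audio analysis.
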SoundCue /BRCosmetics/Sounds/Emotes/Journey/EmoteFoley_Journey_Blade_Equip_1P
root=QualityLevel_0
nodes [
 {id=QualityLevel_0, children [Random_0,Random_0,Random_1,Random_1]}
 {id=Random_0, children [WavePlayer_0,WavePlayer_1,WavePlayer_2,WavePlayer_3,WavePlayer_4,WavePlayer_5], properties {weights [1,1,1,1,1,1]}}
 {id=Random_1, children [WavePlayer_4,WavePlayer_3], properties {weights [1,1]}}
 {id=WavePlayer_0, properties {soundWaveAssetPtr=&/Emote_Journey_Blade_Extend_Out_06}}
 {id=WavePlayer_1, properties {soundWaveAssetPtr=&/Emote_Journey_Blade_Extend_Out_01}}
 {id=WavePlayer_2, properties {soundWaveAssetPtr=&/Emote_Journey_Blade_Extend_Out_02}}
 {id=WavePlayer_3, properties {soundWaveAssetPtr=&/Emote_Journey_Blade_Extend_Out_03}}
 {id=WavePlayer_4, properties {soundWaveAssetPtr=&/Emote_Journey_Blade_Extend_Out_04}}
 {id=WavePlayer_5, properties {soundWaveAssetPtr=&/Emote_Journey_Blade_Extend_Out_05}}
]
properties {
 attenuationSettings=&Fort_Emotes_Athena_Att
 audioPropertiesSheet=/BRCosmetics/Sounds/Settings/PropertySheets/Emotes/APS_Cosmetic_Emote_Foley
 bOverrideConcurrency=true
 maxDistance=1600
 soundClassObject=/Game/Sounds/FortSoundClassesAndModes/Fort_Foley_Emote
 volumeMultiplier=1
 concurrencyOverrides {maxCount=3, resolutionRule=StopOldest}
 concurrencySet [/Game/Sounds/Concurrency_Rules/Athena_EmoteMusic_Conc]
 soundSubmixSends [{minSendLevel=0, soundSubmix=&Fortnite_AudioAnalysis_Music}]
}
